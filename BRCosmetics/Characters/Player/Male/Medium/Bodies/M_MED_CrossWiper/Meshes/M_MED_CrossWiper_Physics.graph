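# Physics asset for the CrossWiper medium male player body, built for secondary motion of its long vest.
# Capsules on the pelvis, spine and thighs form the body, and seven front, side and back vest bone chains hang from spine_04.
# Chains use tight, mostly single-axis angular limits that differ between left and right.
# Free-rotation constraints tie neighboring chain tips around the back and sides, leaving the front open.
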PhysicsAsset /BRCosmetics/Characters/Player/Male/Medium/Bodies/M_MED_CrossWiper/Meshes/M_MED_CrossWiper_Physics
bodies {
 dyn_vest_bk_l_01 [{shapeType=Capsule, length=9.5, radius=7.5, shapeName=dyn_vest_bk_l_01_capsule, center=<5.1483 0 5.1656>, rotation=<rot 0 0 90>}]
 dyn_vest_bk_l_02 [{shapeType=Capsule, length=13.1731, radius=7.5752, shapeName=dyn_vest_bk_l_02_capsule, center=<7.9082 0.712 5.1656>, rotation=<rot 0 0 90>}]
 dyn_vest_bk_l_base [{shapeType=Capsule, length=0.505, radius=0.505, shapeName=dyn_vest_bk_l_base_capsule}]
 dyn_vest_bk_mid_01 [{shapeType=Capsule, length=15, radius=7, shapeName=dyn_vest_bk_mid_01_capsule, center=<6.0812 -5.7161 0>}]
 dyn_vest_bk_mid_02 [{shapeType=Capsule, length=13.7705, radius=8.9294, shapeName=dyn_vest_bk_mid_02_capsule, center=<6.89 -5.7161 0>}]
 dyn_vest_bk_mid_base [{shapeType=Capsule, length=0.505, radius=0.505, shapeName=dyn_vest_bk_mid_base_capsule}]
 dyn_vest_bk_r_01 [{shapeType=Capsule, length=9.5, radius=7.5, shapeName=dyn_vest_bk_r_01_capsule, center=<-5.1483 0 -5.1656>, rotation=<rot 0 0 -90>}]
 dyn_vest_bk_r_02 [{shapeType=Capsule, length=9.3665, radius=7.6335, shapeName=dyn_vest_bk_r_02_capsule, center=<-6.7526 0 -5.1656>, rotation=<rot 0 0 -90>}]
 dyn_vest_bk_r_base [{shapeType=Capsule, length=0.505, radius=0.505, shapeName=dyn_vest_bk_r_base_capsule}]
 dyn_vest_fr_l_01 [{shapeType=Capsule, length=12.0308, radius=5.4692, shapeName=dyn_vest_fr_l_01_capsule, center=<3.9264 0 2.6881>, rotation=<rot 0 0 90>}]
 dyn_vest_fr_l_02 [{shapeType=Capsule, length=11.4426, radius=6.0574, shapeName=dyn_vest_fr_l_02_capsule, center=<3.9264 0 3.5249>, rotation=<rot 0 0 90>}]
 dyn_vest_fr_l_03 [{shapeType=Capsule, length=14, radius=7.5, shapeName=dyn_vest_fr_l_03_capsule, center=<5.5167 0 5.0225>, rotation=<rot 0 0 90>}]
 dyn_vest_fr_l_base [{shapeType=Capsule, length=0.505, radius=0.505, shapeName=dyn_vest_fr_l_base_capsule}]
 dyn_vest_fr_r_01 [{shapeType=Capsule, length=9.8506, radius=7.6494, shapeName=dyn_vest_fr_r_01_capsule, center=<-4.7672 0 -2.6881>, rotation=<rot 0 0 -90>}]
 dyn_vest_fr_r_02 [{shapeType=Capsule, length=11.3, radius=8.7, shapeName=dyn_vest_fr_r_02_capsule, center=<-6.2532 0 -5.0274>, rotation=<rot 0 0 -90>}]
 dyn_vest_fr_r_03 [{shapeType=Capsule, length=9, radius=7.5, shapeName=dyn_vest_fr_r_03_capsule, center=<-5.5167 0 -5.015>, rotation=<rot 0 0 -90>}]
 dyn_vest_fr_r_base [{shapeType=Capsule, length=0.505, radius=0.505, shapeName=dyn_vest_fr_r_base_capsule}]
 dyn_vest_side_l_01 [{shapeType=Capsule, length=11.5, radius=5, shapeName=dyn_vest_side_l_01_capsule, center=<3.9264 0 4.4096>, rotation=<rot 0 0.0003 90>}]
 dyn_vest_side_l_02 [{shapeType=Capsule, length=13.8111, radius=7.4092, shapeName=dyn_vest_side_l_02_capsule, center=<6.5227 0 5.7625>, rotation=<rot 0 0 90>}]
 dyn_vest_side_l_base [{shapeType=Capsule, length=0.505, radius=0.505, shapeName=dyn_vest_side_l_base_capsule}]
 dyn_vest_side_r_01 [{shapeType=Capsule, length=11.5, radius=5, shapeName=dyn_vest_side_r_01_capsule, center=<-3.9264 0 -4.4096>, rotation=<rot 0 0 -90>}]
 dyn_vest_side_r_02 [{shapeType=Capsule, length=9.0768, radius=7.4232, shapeName=dyn_vest_side_r_02_capsule, center=<-6.5227 0 -5.7625>, rotation=<rot 0 0 -90>}]
 dyn_vest_side_r_base [{shapeType=Capsule, length=0.505, radius=0.505, shapeName=dyn_vest_side_r_base_capsule}]
 pelvis [{shapeType=Capsule, length=8.0868, radius=13.6689, center=<-3.9247 -0.4789 0>}]
 spine_01 [{shapeType=Capsule, length=8.5389, radius=12.1711, shapeName=spine_01_capsule, center=<-1.5536 0.1677 0>}]
 spine_02 [{shapeType=Capsule, length=13.0908, radius=10.1508, shapeName=spine_02_capsule, center=<0 -0.9486 0>}]
 spine_03 [{shapeType=Capsule, length=9.0204, radius=11.3509, center=<0 -1.0039 0>}]
 spine_04 [{shapeType=Capsule, length=10, radius=11.9163}]
 thigh_l [{shapeType=Capsule, length=41.495, radius=10.4236, center=<-20.4282 0.9561 0>, rotation=<rot 90 0 0>}]
 thigh_r [{shapeType=Capsule, length=41.495, radius=10.4236, center=<20.4282 -0.9561 0>, rotation=<rot 90 180 0>}]
}
constraints [
 {bone1Name=spine_03, bone2Name=spine_02, swing1LimitDegrees=45, swing2LimitDegrees=45, twistLimitDegrees=45}
 {bone1Name=spine_04, bone2Name=spine_03, swing1LimitDegrees=45, swing2LimitDegrees=45, twistLimitDegrees=45}
 {bone1Name=thigh_l, bone2Name=pelvis, swing1LimitDegrees=45, swing2LimitDegrees=45, twistLimitDegrees=45}
 {bone1Name=thigh_r, bone2Name=pelvis, swing1LimitDegrees=45, swing2LimitDegrees=45, twistLimitDegrees=45}
 {
  bone1Name=spine_02
  bone2Name=spine_01
  swing1LimitDegrees=45
  swing1Motion=Free
  swing2LimitDegrees=45
  swing2Motion=Free
  twistLimitDegrees=45
  twistMotion=Free
 }
 {
  bone1Name=spine_01
  bone2Name=pelvis
  swing1LimitDegrees=45
  swing1Motion=Free
  swing2LimitDegrees=45
  swing2Motion=Free
  twistLimitDegrees=45
  twistMotion=Free
 }
 {bone1Name=dyn_vest_bk_l_base, bone2Name=spine_04, swing1LimitDegrees=0, swing2LimitDegrees=5, twistLimitDegrees=0}
 {bone1Name=dyn_vest_bk_mid_base, bone2Name=spine_04, swing1LimitDegrees=5, swing2LimitDegrees=0, twistLimitDegrees=0}
 {bone1Name=dyn_vest_bk_r_base, bone2Name=spine_04, swing1LimitDegrees=0, swing2LimitDegrees=5, twistLimitDegrees=0}
 {bone1Name=dyn_vest_fr_l_base, bone2Name=spine_04, swing1LimitDegrees=0, swing2LimitDegrees=5, twistLimitDegrees=0}
 {bone1Name=dyn_vest_fr_r_base, bone2Name=spine_04, swing1LimitDegrees=0, swing2LimitDegrees=5, twistLimitDegrees=0}
 {bone1Name=dyn_vest_side_l_base, bone2Name=spine_04, swing1LimitDegrees=0, swing2LimitDegrees=20, twistLimitDegrees=0}
 {bone1Name=dyn_vest_side_r_base, bone2Name=spine_04, swing1LimitDegrees=0, swing2LimitDegrees=20, twistLimitDegrees=0}
 {bone1Name=dyn_vest_bk_l_01, bone2Name=dyn_vest_bk_l_base, swing1LimitDegrees=0, swing2LimitDegrees=5, twistLimitDegrees=0}
 {bone1Name=dyn_vest_bk_l_02, bone2Name=dyn_vest_bk_l_01, swing1LimitDegrees=0, swing2LimitDegrees=5, twistLimitDegrees=0}
 {bone1Name=dyn_vest_bk_mid_01, bone2Name=dyn_vest_bk_mid_base, swing1LimitDegrees=5, swing2LimitDegrees=0, twistLimitDegrees=0}
 {bone1Name=dyn_vest_bk_mid_02, bone2Name=dyn_vest_bk_mid_01, swing1LimitDegrees=5, swing2LimitDegrees=0, twistLimitDegrees=0}
 {bone1Name=dyn_vest_bk_r_01, bone2Name=dyn_vest_bk_r_base, swing1LimitDegrees=0, swing2LimitDegrees=5, twistLimitDegrees=0}
 {bone1Name=dyn_vest_bk_r_02, bone2Name=dyn_vest_bk_r_01, swing1LimitDegrees=0, swing2LimitDegrees=5, twistLimitDegrees=0}
 {bone1Name=dyn_vest_fr_l_01, bone2Name=dyn_vest_fr_l_base, swing1LimitDegrees=0, swing2LimitDegrees=20, twistLimitDegrees=0}
 {bone1Name=dyn_vest_fr_l_02, bone2Name=dyn_vest_fr_l_01, swing1LimitDegrees=0, swing2LimitDegrees=30, twistLimitDegrees=5}
 {bone1Name=dyn_vest_fr_l_03, bone2Name=dyn_vest_fr_l_02, swing1LimitDegrees=0, swing2LimitDegrees=30, twistLimitDegrees=5}
 {bone1Name=dyn_vest_fr_r_01, bone2Name=dyn_vest_fr_r_base, swing1LimitDegrees=0, swing2LimitDegrees=5, twistLimitDegrees=0}
 {bone1Name=dyn_vest_fr_r_02, bone2Name=dyn_vest_fr_r_01, swing1LimitDegrees=0, swing2LimitDegrees=10, twistLimitDegrees=5}
 {bone1Name=dyn_vest_fr_r_03, bone2Name=dyn_vest_fr_r_02, swing1LimitDegrees=0, swing2LimitDegrees=20, twistLimitDegrees=5}
 {bone1Name=dyn_vest_side_l_01, bone2Name=dyn_vest_side_l_base, swing1LimitDegrees=20, swing2LimitDegrees=20, twistLimitDegrees=5}
 {bone1Name=dyn_vest_side_l_02, bone2Name=dyn_vest_side_l_01, swing1LimitDegrees=45, swing2LimitDegrees=40, twistLimitDegrees=40}
 {bone1Name=dyn_vest_side_r_01, bone2Name=dyn_vest_side_r_base, swing1LimitDegrees=0, swing2LimitDegrees=5, twistLimitDegrees=0}
 {bone1Name=dyn_vest_side_r_02, bone2Name=dyn_vest_side_r_01, swing1LimitDegrees=10, swing2LimitDegrees=10, twistLimitDegrees=0}
 {
  bone1Name=dyn_vest_fr_l_03
  bone2Name=dyn_vest_side_l_02
  swing1LimitDegrees=45
  swing1Motion=Free
  swing2LimitDegrees=45
  swing2Motion=Free
  twistLimitDegrees=45
  twistMotion=Free
 }
 {
  bone1Name=dyn_vest_fr_r_03
  bone2Name=dyn_vest_side_r_02
  swing1LimitDegrees=45
  swing1Motion=Free
  swing2LimitDegrees=45
  swing2Motion=Free
  twistLimitDegrees=45
  twistMotion=Free
 }
 {
  bone1Name=dyn_vest_side_l_02
  bone2Name=dyn_vest_bk_l_02
  swing1LimitDegrees=45
  swing1Motion=Free
  swing2LimitDegrees=45
  swing2Motion=Free
  twistLimitDegrees=45
  twistMotion=Free
 }
 {
  bone1Name=dyn_vest_bk_l_02
  bone2Name=dyn_vest_bk_mid_02
  swing1LimitDegrees=45
  swing1Motion=Free
  swing2LimitDegrees=45
  swing2Motion=Free
  twistLimitDegrees=45
  twistMotion=Free
 }
 {
  bone1Name=dyn_vest_side_r_02
  bone2Name=dyn_vest_bk_r_02
  swing1LimitDegrees=45
  swing1Motion=Free
  swing2LimitDegrees=45
  swing2Motion=Free
  twistLimitDegrees=45
  twistMotion=Free
 }
 {
  bone1Name=dyn_vest_bk_r_02
  bone2Name=dyn_vest_bk_mid_02
  swing1LimitDegrees=45
  swing1Motion=Free
  swing2LimitDegrees=45
  swing2Motion=Free
  twistLimitDegrees=45
  twistMotion=Free
 }
]
