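# Animation Blueprint for the M_MED_CrossWiper player body cosmetic, layering procedural rigging and secondary physics onto the character's pose.
# It copies the attached parent mesh's pose into the CrossWiper Control Rig, which receives an is-in-frontend flag.
# The cached result feeds a RigidBody node simulating the physics asset relative to spine_04, with tuned drag, velocity and acceleration limits.
# The update event has no logic.
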AnimBlueprint /BRCosmetics/Characters/Player/Male/Medium/Bodies/M_MED_CrossWiper/Meshes/M_MED_CrossWiper_AnimBP
target_skeleton=&/M_MED_CrossWiper_Skeleton
graphs [
 AnimGraph {nodes [
   n1 = RigidBody(ComponentPose=Local To Component(LocalPose=Use cached pose 'toPhysics'())) {
    bClampLinearTranslationLimitToRefPose=true
    bOverrideWorldGravity=true
    overridePhysicsAsset=&/M_MED_CrossWiper_Physics
    simulationSpace=BaseBoneSpace
    baseBoneRef {boneName=spine_04}
    overrideWorldGravity {z=-980}
    simSpaceSettings {
     dampingAlpha=0
     maxAngularAcceleration=40
     maxAngularVelocity=8
     maxLinearAcceleration=700
     maxLinearVelocity=40
     worldAlpha=1
     externalLinearDragV=<0.7 0.4 0.4>
    }
   }
   Output Pose(Result=Component To Local(ComponentPose=@n1)) {layerGroup=None}
   n2 = Control Rig!(Source=Copy Pose From Mesh() {bUseAttachedParent=true}, Alpha, is Frontend:bool=Get bIsInFrontEnd()) -> (Pose) {controlRigAssetReference {blueprintRigClass=&/M_MED_CrossWiper_CtrlRig.M_MED_CrossWiper_CtrlRig_C}}
   Save cached pose 'toPhysics'(Pose=@n2) {cacheName=toPhysics}
  ]}
 EventGraph {nodes [Event BlueprintUpdateAnimation()]}
]
variables []
